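; Gameplay ability for the CoveSpice weapon that chains a shock to nearby player pawns when triggered by a gameplay event.
; It scans around the event's pawn within Chain Scan Range, excluding only the initial instigator and the event target.
; Each remaining pawn receives the Shock GE and Damage GE.
; The damage context names the event target as instigator and origin, carrying the initial instigator's current weapon.
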(event Ability|EventActivateAbilityFromEvent
  (params (EventData :type Gameplay%20Event%20Data%20Structure%20%28by%20ref%29))
  (|CallParentFunction EventData)
  (bind _output_get (Variables|Default|SetActivation%20Event%20Data EventData))
  (bind _asfort_pawn (Utilities|Casting|CastToFortPawn (Utilities|Struct|BreakGameplayEventData _output_get)))
  (bind _outfortplayerpawns (Fortnite|GetFortPlayerPawnsInRange (Transformation|GetWorldLocation (Variables|/Script/Engine.Character|GetCapsuleComponent _asfort_pawn)) (ScalableFloat|GetValueatLevel (Variables|Default|GetChain%20Scan%20Range))))
  (for _array_element _outfortplayerpawns
    (if (CallFunction|CanChainTo _array_element)
      (Environment|ApplyGlobalEnvironmentGameplayEffecttoActor _array_element (Variables|Default|GetShock%20GE))
      (CallFunction|ApplyChainedDamage _array_element))))

(fn ApplyChainedDamage
  (params (Player%20Pawn%20to%20Damage :type Fort%20Player%20Pawn%20Object%20Reference))
  (bind _activation_event_data (Variables|Default|GetActivation%20Event%20Data))
  (bind (_instigator _target) (Utilities|Struct|BreakGameplayEventData _activation_event_data))
  (bind _asfort_player_state_athena (Utilities|Casting|CastToFortPlayerStateAthena _instigator))
  (LocalVariables|SetInitial%20Instigator (Utilities|Casting|CastToFortPlayerPawn (PlayerState|GetPawn _asfort_player_state_athena)))
  (bind _output_get (LocalVariables|SetActual%20Instigator _target))
  (bind _output_get_2 (LocalVariables|SetEffect%20Context (Class|AthenaFunctionLibrary|MakeSimpleFortEffectContext _output_get (Weapon|GetCurrentWeapon (LocalVariables|GetInitial%20Instigator)))))
  (Ability|EffectContext|SetOrigin _output_get_2 (Transformation|GetActorLocation _output_get))
  (GameplayEffects|ApplyGameplayEffectToSelf (Ability|GetAbilitySystemComponent (Variables|GetPlayer%20Pawn%20to%20Damage)) (Variables|Default|GetDamage%20GE) 0.0 (LocalVariables|GetEffect%20Context)))

(fn CanChainTo
  (params (Target%20Pawn :type Actor%20Object%20Reference))
  (bind _activation_event_data (Variables|Default|GetActivation%20Event%20Data))
  (bind (_instigator _target) (Utilities|Struct|BreakGameplayEventData _activation_event_data))
  (Utilities|FlowControl|Sequence
    (:then_0
      (bind _asfort_player_state_athena (Utilities|Casting|CastToFortPlayerStateAthena _instigator))
      (LocalVariables|SetInitial%20Instigator (PlayerState|GetPawn _asfort_player_state_athena))
      (LocalVariables|SetActual%20Instigator _target)
      (if (and (!= (Variables|GetTarget%20Pawn) (LocalVariables|GetActual%20Instigator)) (!= (Variables|GetTarget%20Pawn) (LocalVariables|GetInitial%20Instigator)))
        (LocalVariables|SetCan%20Chain%20to%20Target true)
        (else
          (LocalVariables|SetCan%20Chain%20to%20Target false))))
    (:then_1
      (return (LocalVariables|GetCan%20Chain%20to%20Target)))))
