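; Event-triggered gameplay ability for the WildEstate CoveSpice weapon that chains a damaging shock to multiple targets.
; On activation it stores the event data and, if no targeting task is stored yet, runs the ChainShock targeting preset.
; Targeted actors receive shock and damage effects plus a burst cue, with context from the event's pawn and its current weapon.
; The loop count comes from the Max Chained Enemies scalable float, without clamping to the target count.
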(event Ability|EventActivateAbilityFromEvent
  (params (EventData :type Gameplay%20Event%20Data%20Structure%20%28by%20ref%29))
  (|CallParentFunction EventData)
  (Variables|Default|SetEvent%20Data EventData)
  (CallFunction|ChainTarget))

(event Custom|Chain%20Target
  (Variables|Default|GetTargeting%20Task
    (:then)
    (:else
      (bind (_asynctaskproxy _targetdata) (Ability|Tasks|PerformTargetingAbilityTask "None" "/WildEstateWeapons/Gameplay/CoveSpice/TP_CoveSpice_ChainShock.TP_CoveSpice_ChainShock")
        (:then
          (Variables|Default|SetTargeting%20Task _asynctaskproxy))
        (:Targeted
          (CallFunction|ApplyShocktoTargets (Ability|TargetData|GetAllActorsfromTargetData _targetdata)))))))

(fn ApplyShockToTargets
  (params (Actors :type Array%20of%20Actor%20Object%20References))
  (bind _targets (LocalVariables|GetTargets))
  (bind _effect_context (LocalVariables|GetEffect%20Context))
  (LocalVariables|SetTargets Actors)
  (bind _asfort_pawn (Utilities|Casting|CastToFortPawn (Utilities|Struct|BreakGameplayEventData (Variables|Default|GetEvent%20Data))))
  (bind _output_get (LocalVariables|SetEffect%20Context (Class|AthenaFunctionLibrary|MakeSimpleFortEffectContext _asfort_pawn (Weapon|GetCurrentWeapon _asfort_pawn))))
  (Ability|EffectContext|SetOrigin _output_get (Transformation|GetWorldLocation (Variables|/Script/Engine.Character|GetCapsuleComponent _asfort_pawn)))
  (for _index (range (ScalableFloat|GetValueatLevel (Variables|Default|GetMax%20Chained%20Enemies)))
    (Fortnite|ApplyGameplayEffectsToActor (Utilities|Array|Get(acopy) _targets _index) (Utilities|Array|MakeArray (Variables|Default|GetChain%20Shock%20GE) (Variables|Default|GetDamage%20GE)) (Transformation|GetWorldLocation (Variables|/Script/Engine.Character|GetCapsuleComponent _asfort_pawn)) true 0 (Ability|GetAbilitySystemComponent _asfort_pawn) _effect_context)
    (GameplayCue|ExecuteGameplayCueOnActor(Burst) (Utilities|Array|Get(acopy) _targets _index) (Variables|Default|GetChain%20Shock%20Cue) (Ability|GameplayCue|MakeGameplayCueParameters :NormalizedMagnitude 0.0 :RawMagnitude 0.0 :EffectContext _effect_context :Location (Transformation|GetWorldLocation (Variables|/Script/Engine.Character|GetCapsuleComponent _asfort_pawn)) :Normal "0,0,0" :Instigator (Ability|GetAvatarActorfromActorInfo) :EffectCauser _asfort_pawn :GameplayEffectLevel 1 :AbilityLevel 1 :bReplicateLocationWhenUsingMinimalRepProxy false))))
